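# Material function for destroyed-tree leaf particles that adds motion and a dithered dissolve to incoming material attributes.
# World position offset pushes vertices outward from particle center with vertex-color variation and shifts them along Z by vertex blue, driven by WPO_Expand and WPO_Drop dynamic parameters.
# Opacity mask is multiplied by a contrast-adjusted temporal dither built from vertex color, a smooth-step texture and particle alpha.
# A camera-centered sphere mask fades out nearby leaves.
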MaterialFunction /Game/Effects/Fort_Effects/World/Destruction/Trees/TreeDestruction_New/MF_LeavesWPO
graph {nodes [
  n1 = WorldPosition()
  n2 = ParticlePositionWS()
  n3 = VertexColor()
  n4 = Add(A=Normalize(VectorInput=Subtract(A=@n1.XYZ, B=@n2)), B=ConstantBiasScale(None=@n3) {bias=-0.5, scale=4})
  n5 = DynamicParameter() {paramNames [WPO_Expand, WPO_Drop, "", Param4]}
  n6 = ScalarParameter() {defaultValue=-0.1, parameterName=WPO_Z_LowMoving}
  n7 = Multiply(A=Multiply(A=LinearInterpolate(A=@n6, Alpha=@n3.B) {constA=-0.5}) {constB=5}, B=Constant3Vector() {constant=<rgba 0 0 -1 0>})
  n8 = MaterialFunctionCall!(
   Material=FunctionInput() {inputType=FunctionInput_MaterialAttributes}
   WorldPositionOffset=Add(A=Multiply(A=@n4, B=@n5.WPO_Expand), B=Multiply(A=@n7, B=@n5.WPO_Drop))
  ) {materialFunction=/Engine/Functions/MaterialLayerFunctions/MatLayerBlend_AddWorldPositionOffset}
  n9 = GetMaterialAttributes(None=@n8.Blended Material) {attributeGetTypes [#1]}
  n10 = Multiply(A=@n3.R) {constB=2}
  n11 = TextureCoordinate()
  n12 = AppendVector(
   A=ScalarParameter() {defaultValue=0.5, parameterName=Smooth Step U Scale}
   B=ScalarParameter() {defaultValue=0.5, parameterName=Smooth Step V Scale}
  )
  n13 = TextureSampleParameter2D(UVs=Add(A=AppendVector(A=@n3.R, B=@n3.G), B=Multiply(A=@n11, B=@n12))) {parameterName=SmoothStep_Texture, texture=&2ndInfectedMapLinear}
  n14 = WorldPosition()
  n15 = CameraPositionWS()
  n16 = Subtract(
   A=Add(A=Add(A=@n10, B=@n13.R), B=ParticleColor().A)
   B=Multiply(A=SphereMask(A=@n14.XYZ, B=@n15) {attenuationRadius=1000, hardnessPercent=1}) {constB=2}
  )
  n17 = MaterialFunctionCall(Alpha Threshold=MaterialFunctionCall(In=Clamp(None=@n16), Contrast=ScalarParameter() {defaultValue=1, parameterName=AlphaContrast}) {materialFunction=&CheapContrast}.Result) {materialFunction=&DitherTemporalAA_Fortnite}
  FunctionOutput(None=SetMaterialAttributes!(MaterialAttributes=@n8.Blended Material, OpacityMask=Multiply(A=@n9.OpacityMask, B=@n17.Result), Normal) {attributeSetTypes [#1, #2]})
 ]}
