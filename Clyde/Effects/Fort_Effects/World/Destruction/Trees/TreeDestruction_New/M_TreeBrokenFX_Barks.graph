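# Lit mesh-particle material for bark pieces in the tree destruction effect.
# UV channel 2 becomes radial coordinates, with the first component folded into a mirrored ramp, and these drive the bark color and normal samples.
# Constant metallic, specular and roughness join them in material attributes, which pass through MF_LeavesWPO.
# The diffuse, specular-mask and mask texture parameters do not feed the material attributes.
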Material /Clyde/Effects/Fort_Effects/World/Destruction/Trees/TreeDestruction_New/M_TreeBrokenFX_Barks
graph {
 properties {
  bUseMaterialAttributes=true
  bUsedWithMeshParticles=true
  pixelDepthOffsetMode=PDOM_Legacy
  previewMesh=&/Athena_Tree_Large_01_CustomLOD_Bark_BreakMesh
  usedShadingModels=Default Lit
 }
 nodes [
  n1 = MaterialFunctionCall(Vector or UVs=TextureCoordinate() {coordinateIndex=2}) {materialFunction=&VectorToRadialValue}
  n2 = ComponentMask(None=@n1.Radial Coordinates) {r=true}
  n3 = Saturate(None=Multiply(A=@n2) {constB=2})
  n4 = Saturate(None=Multiply(A=OneMinus(None=@n2)) {constB=2})
  n5 = AppendVector(A=Multiply(A=Multiply(A=@n3, B=@n4)) {constB=0.75}, B=ComponentMask(None=@n1.Radial Coordinates) {g=true})
  n6 = TextureSample(UVs=@n5) {texture=/Clyde/Environments/Props/Trees/Textures/T_Bark}
  n7 = TextureSampleParameter2D(UVs=@n5) {group=PX, parameterName=Normals, samplerType=SAMPLERTYPE_Normal, texture=/Game/Environments/Props/Trees/Textures/T_OnBoarding_Tree_01_N}
  MaterialFunctionCall!(In=MakeMaterialAttributes!(
    BaseColor=ComponentMask(None=@n6.RGB) {b=true, desc=BaseColor, g=true, r=true}
    Metallic=Constant()
    Specular=Constant() {r=0.5}
    Roughness=Constant() {r=1}
    Anisotropy
    EmissiveColor
    Opacity
    OpacityMask
    Normal=@n7.RGB
    Tangent
    WorldPositionOffset
    SubsurfaceColor
    ClearCoat
    ClearCoatRoughness
    AmbientOcclusion
    Refraction
    CustomizedUVs_0
    CustomizedUVs_1
    CustomizedUVs_2
    CustomizedUVs_3
    CustomizedUVs_4
    CustomizedUVs_5
    CustomizedUVs_6
    CustomizedUVs_7
    PixelDepthOffset
    ShadingModel
    Displacement
   )) {materialFunction=/Game/Effects/Fort_Effects/World/Destruction/Trees/TreeDestruction_New/MF_LeavesWPO}
  n8 = TextureSampleParameter2D() {
   desc="spec, metallic,roughness\n"
   group=Specular
   parameterName=SpecularMasks
   samplerType=SAMPLERTYPE_Masks
   texture=/Game/Environments/Props/Trees/Textures/T_OnBoarding_Tree_01_S
  }
  Reroute(None=MaterialFunctionCall(Float3=@n8.RGB) {desc=SpecularTexutre, materialFunction=&BreakOutFloat3Components}.G)
  n9 = TextureSampleParameter2D() {
   group=Textures
   parameterName=MaskTexture
   samplerType=SAMPLERTYPE_Masks
   texture=&T_OnBoarding_Tree_01_S
  }
  n10 = MaterialFunctionCall(Float3=@n9.RGB) {desc=Mask Texture, materialFunction=&BreakOutFloat3Components}
  Reroute(None=@n10.B)
  Reroute(None=@n10.R)
  Reroute(None=ComponentMask(None=@n7.RGB) {b=true, desc=Normal, g=true, r=true})
  Reroute(None=@n10.G)
  TextureSampleParameter2D() {group=Textures, parameterName=Diffuse, texture=/Game/Environments/Props/Trees/Textures/T_OnBoarding_Tree_01_D}
 ]
}
